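; BP_ShipRamp is a ship ramp actor Blueprint, located in a test map folder, whose lower section has an adjustable scale.
; The construction script sets the world scale of RampBottom, driving Y from the SecondaryRampScale variable and holding X and Z at a fixed value.
; BeginPlay, ActorBeginOverlap and Tick are empty stubs, so the graph contains no runtime logic.
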(event EventBeginPlay)

(event Collision|EventActorBeginOverlap
  (params (OtherActor :type Actor%20Object%20Reference)))

(event EventTick
  (params (DeltaSeconds :type Float%20%28single-precision%29)))

(fn ConstructionScript
  (Transformation|SetWorldScale3D (Variables|Default|GetRampBottom) (Math|Vector|MakeVector 0.8 (Variables|Default|GetSecondaryRampScale) 0.8)))
