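; BP_ShipSupport is an actor Blueprint for a ship support leg with an adjustable lower leg and foot, found in a test map folder.
; The construction script rotates ShipLeg_Bottom by the clamped ShipLegSupportAngle.
; It then snaps ShipLeg_Foot to that leg's "Bottom" socket in actor space and tilts it by the clamped ShipFootTiltAngle.
; BeginPlay, overlap and Tick events are empty, so this graph has no runtime logic.
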(event EventBeginPlay)

(event Collision|EventActorBeginOverlap
  (params (OtherActor :type Actor%20Object%20Reference)))

(event EventTick
  (params (DeltaSeconds :type Float%20%28single-precision%29)))

(fn ConstructionScript
  (Transformation|SetRelativeRotation (Variables|Default|GetShipLeg_Bottom) (Math|Rotator|MakeRotator 0.0 (Utilities|ClampMinFloat (Utilities|ClampMaxFloat (Variables|Default|GetShipLegSupportAngle) 25.0) -3e+01)) false false)
  (Transformation|SetRelativeLocationAndRotation (Variables|Default|GetShipLeg_Foot) (.location (Transformation|GetSocketTransform (Variables|Default|GetShipLeg_Bottom) "Bottom" "RTS_Actor")) (Math|Rotator|MakeRotator (Utilities|ClampMinFloat (Utilities|ClampMaxFloat (Variables|Default|GetShipFootTiltAngle) 1e+01) -1e+01)) false false))
